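# Editor utility widget for batch-generating Sparks song item definitions over a Start ID to End ID range.
# Property views bind to SparksSongTemplatesEditorSubsystem to expose its settings, and to the widget itself for the ID range.
# On Construct it sets the subsystem's current release version and fills any unset song template and global jam emote references with placeholders.
# Generate calls CreateMultipleSongItemDefs; a collapsed legacy button calls UpdateAllSongsAndEmoteItemDefs.
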EditorUtilityWidgetBlueprint /SparksSongTemplates/EUW_GenerateSparksSongItemDefs
graphs [
 EventGraph {nodes [
   "On Clicked (EditorUtilityButton_GenerateSongItemDefs)"() {is=title K2Node_ComponentBoundEvent}
   n1 = Get SparksSongTemplatesEditorSubsystem() -> (ReturnValue:Sparks Song Templates Editor Subsystem Object Reference) {is=title K2Node_GetEditorSubsystem}
   > CreateMultipleSongItemDefs!(
    execute
    self:Sparks Song Templates Editor Subsystem Object Reference=@n1
    StartId:int=Get Start ID() -> (Start ID:int=0)
    EndId:int=Get End ID() -> (End ID:int=0)
   ) -> (then, ReturnValue:Array of Sparks Song Item Definition Object References)
   "On Clicked (EditorUtilityButton_FixupLegacy)"() {is=title K2Node_ComponentBoundEvent}
   > UpdateAllSongsAndEmoteItemDefs(self:Sparks Song Templates Editor Subsystem Object Reference=@n1)
   Event PreConstruct()
   > SetObject(
    self=[
     Get SinglePropertyView_SongItem_NameFormat() -> (SinglePropertyView_SongItem_NameFormat:Single Property View Object Reference)
     Get CustomPropertyView_326() -> (CustomPropertyView_326:Custom Property View Object Reference)
     Get SinglePropertyView_ReleaseVersion() -> (SinglePropertyView_ReleaseVersion:Single Property View Object Reference)
     Get SinglePropertyView_SongItemTemplate() -> (SinglePropertyView_SongItemTemplate:Single Property View Object Reference)
     Get SinglePropertyView_SongItemDir() -> (SinglePropertyView_SongItemDir:Single Property View Object Reference)
    ]
    NewObject=Get SparksSongTemplatesEditorSubsystem()
   )
   > SetObject(
    self=[
     Get SinglePropertyView_StartId() -> (SinglePropertyView_StartId:Single Property View Object Reference)
     Get SinglePropertyView_EndId() -> (SinglePropertyView_EndId:Single Property View Object Reference)
    ]
    NewObject=Self-Reference()
   )
   Event Construct()
   > UpdateDefaultValuesIfRequired()
  ]}
 UpdateDefaultValuesIfRequired {nodes [
   UpdateDefaultValuesIfRequired() {is=title K2Node_FunctionEntry}
   > Set SongTemplatesSubsystem(SongTemplatesSubsystem:Sparks Song Templates Editor Subsystem Object Reference=Get SparksSongTemplatesEditorSubsystem()) -> (Output_Get:Sparks Song Templates Editor Subsystem Object Reference)
   > n1 = EditorOnly_GetCurrentFortReleaseVersion!(execute) -> (then, ReturnValue:Fort Release Version Structure)
   > Set FortReleaseVersion!(
    execute
    FortReleaseVersion:Fort Release Version Structure=@n1
    self:Sparks Song Templates Editor Subsystem Object Reference=Get SongTemplatesSubsystem() -> (SongTemplatesSubsystem:Sparks Song Templates Editor Subsystem Object Reference)
   ) -> (then, Output_Get:Fort Release Version Structure)
   > n2 = Sequence!(execute) -> (then_0, then_1, then_2:Exec, then_3:Exec, then_4:Exec, then_5:Exec)
   n3 = Get SongTemplatesSubsystem()
   n4 = Is Valid(
    exec=@n2.then_1
    InputObject=Get JamSongPart_GlobalEID_All!(self:Sparks Song Templates Editor Subsystem Object Reference=@n3) -> (JamSongPart_GlobalEID_All:Sparks Jam Emote Item Definition Object Reference)
   )
   Set JamSongPart_GlobalEID_All!(
    execute=@n4.Is Not Valid
    JamSongPart_GlobalEID_All:Sparks Jam Emote Item Definition Object Reference=/SparksSongTemplates/Items/JamEmotes/EID_Placeholder_Global_None
    self:Sparks Song Templates Editor Subsystem Object Reference=@n3
   ) -> (then, Output_Get:Sparks Jam Emote Item Definition Object Reference)
   n5 = Get SongTemplatesSubsystem()
   n6 = Is Valid(
    exec=@n2.then_2
    InputObject=Get JamSongPart_GlobalEID_Bass!(self:Sparks Song Templates Editor Subsystem Object Reference=@n5) -> (JamSongPart_GlobalEID_Bass:Sparks Jam Emote Item Definition Object Reference)
   )
   Set JamSongPart_GlobalEID_Bass!(
    execute=@n6.Is Not Valid
    JamSongPart_GlobalEID_Bass:Sparks Jam Emote Item Definition Object Reference=/SparksSongTemplates/Items/JamEmotes/EID_Placeholder_Global_Bass
    self:Sparks Song Templates Editor Subsystem Object Reference=@n5
   ) -> (then, Output_Get:Sparks Jam Emote Item Definition Object Reference)
   n7 = Get SongTemplatesSubsystem()
   n8 = Is Valid(
    exec=@n2.then_3
    InputObject=Get JamSongPart_GlobalEID_Drum!(self:Sparks Song Templates Editor Subsystem Object Reference=@n7) -> (JamSongPart_GlobalEID_Drum:Sparks Jam Emote Item Definition Object Reference)
   )
   Set JamSongPart_GlobalEID_Drum!(
    execute=@n8.Is Not Valid
    JamSongPart_GlobalEID_Drum:Sparks Jam Emote Item Definition Object Reference=/SparksSongTemplates/Items/JamEmotes/EID_Placeholder_Global_Drum
    self:Sparks Song Templates Editor Subsystem Object Reference=@n7
   ) -> (then, Output_Get:Sparks Jam Emote Item Definition Object Reference)
   n9 = Get SongTemplatesSubsystem()
   n10 = Is Valid(
    exec=@n2.then_4
    InputObject=Get JamSongPart_GlobalEID_Lead!(self:Sparks Song Templates Editor Subsystem Object Reference=@n9) -> (JamSongPart_GlobalEID_Lead:Sparks Jam Emote Item Definition Object Reference)
   )
   Set JamSongPart_GlobalEID_Lead!(
    execute=@n10.Is Not Valid
    JamSongPart_GlobalEID_Lead:Sparks Jam Emote Item Definition Object Reference=/SparksSongTemplates/Items/JamEmotes/EID_Placeholder_Global_Lead
    self:Sparks Song Templates Editor Subsystem Object Reference=@n9
   ) -> (then, Output_Get:Sparks Jam Emote Item Definition Object Reference)
   n11 = Get SongTemplatesSubsystem()
   n12 = Is Valid(
    exec=@n2.then_0
    InputObject=Get SongItemTemplate!(self:Sparks Song Templates Editor Subsystem Object Reference=@n11) -> (SongItemTemplate:Sparks Song Item Definition Object Reference)
   )
   Set SongItemTemplate!(
    execute=@n12.Is Not Valid
    SongItemTemplate:Sparks Song Item Definition Object Reference=/SparksSongTemplates/Templates/SID_Placeholder_TEMPLATE
    self:Sparks Song Templates Editor Subsystem Object Reference=@n11
   ) -> (then, Output_Get:Sparks Song Item Definition Object Reference)
   n13 = Get SongTemplatesSubsystem()
   n14 = Is Valid(
    exec=@n2.then_5
    InputObject=Get JamSongPart_GlobalEID_Vox!(self:Sparks Song Templates Editor Subsystem Object Reference=@n13) -> (JamSongPart_GlobalEID_Vox:Sparks Jam Emote Item Definition Object Reference)
   )
   Set JamSongPart_GlobalEID_Vox!(
    execute=@n14.Is Not Valid
    JamSongPart_GlobalEID_Vox:Sparks Jam Emote Item Definition Object Reference=/SparksSongTemplates/Items/JamEmotes/EID_Placeholder_Global_Vox
    self:Sparks Song Templates Editor Subsystem Object Reference=@n13
   ) -> (then, Output_Get:Sparks Jam Emote Item Definition Object Reference)
  ]}
]
variables [
 End ID {type=integer}
 Start ID {type=integer}
]
widgets {
 info {rootWidgetClass=&EditorUtilityScrollBox, widgetCount=15}
 widgets [
  EditorUtilityScrollBox_0 {widgetClassPath=&EditorUtilityScrollBox, widget=EditorUtilityScrollBox_0}
  StackBox_0 {
   widgetClassPath=&StackBox
   parent=EditorUtilityScrollBox_0
   widget=StackBox_0
   properties {orientation=Orient_Vertical}
   slot {class=ScrollBoxSlot}
  }
  IndividualPropertyViews {
   widgetClassPath=&VerticalBox
   parent=StackBox_0
   widget=IndividualPropertyViews
   properties {visibility=Collapsed}
   slot {class=StackBoxSlot}
  }
  SinglePropertyView_SongItemDir {
   widgetClassPath=&SinglePropertyView
   parent=IndividualPropertyViews
   widget=SinglePropertyView_SongItemDir
   properties {propertyName=SongItemsDirectory}
   slot {class=VerticalBoxSlot}
  }
  SinglePropertyView_SongItemTemplate {
   widgetClassPath=&SinglePropertyView
   parent=IndividualPropertyViews
   widget=SinglePropertyView_SongItemTemplate
   properties {propertyName=SongItemTemplate}
   slot {class=VerticalBoxSlot}
  }
  SinglePropertyView_SongItem_NameFormat {
   widgetClassPath=&SinglePropertyView
   parent=IndividualPropertyViews
   widget=SinglePropertyView_SongItem_NameFormat
   properties {propertyName=SongItemNameFormat}
   slot {class=VerticalBoxSlot}
  }
  SinglePropertyView_ReleaseVersion {
   widgetClassPath=&SinglePropertyView
   parent=IndividualPropertyViews
   widget=SinglePropertyView_ReleaseVersion
   properties {propertyName=FortReleaseVersion}
   slot {class=VerticalBoxSlot}
  }
  CustomPropertyView_326 {
   widgetClassPath=/Script/FortniteEditor.CustomPropertyView
   bIsVariable=true
   parent=StackBox_0
   widget=CustomPropertyView_326
   properties {
    bShowAllProperties=true
    columns [{columnWidth=300, nameFillWidth=1, valueFillWidth=1}]
    propertiesToExclude [bGenerateJamSongParts]
    propertyGroups [
     {groupName=SparksSongItems, groupColor=<rgba 1 1 1 1>}
     {groupName=JamEmoteItems, groupColor=<rgba 1 1 1 1>}
     {groupName=SparksSongParts, groupColor=<rgba 1 1 1 1>}
    ]
   }
   slot {class=StackBoxSlot}
  }
  StackBox_119 {widgetClassPath=&StackBox, parent=StackBox_0, widget=StackBox_119, slot {class=StackBoxSlot}}
  SinglePropertyView_StartId {
   widgetClassPath=&SinglePropertyView
   parent=StackBox_119
   widget=SinglePropertyView_StartId
   properties {propertyName=Start ID}
   slot {class=StackBoxSlot}
  }
  SinglePropertyView_EndId {
   widgetClassPath=&SinglePropertyView
   parent=StackBox_119
   widget=SinglePropertyView_EndId
   properties {propertyName=End ID}
   slot {class=StackBoxSlot}
  }
  EditorUtilityButton_GenerateSongItemDefs {
   widgetClassPath=&EditorUtilityButton
   parent=StackBox_119
   widget=EditorUtilityButton_GenerateSongItemDefs
   slot {class=StackBoxSlot}
  }
  TextBlock_87 {
   widgetClassPath=&TextBlock
   parent=EditorUtilityButton_GenerateSongItemDefs
   widget=TextBlock_87
   properties {text=Generate New SparksSongItemDefs}
   slot {class=ButtonSlot}
  }
  EditorUtilityButton_FixupLegacy {
   widgetClassPath=&EditorUtilityButton
   parent=StackBox_0
   widget=EditorUtilityButton_FixupLegacy
   properties {visibility=Collapsed}
   slot {class=StackBoxSlot}
  }
  TextBlock_152 {
   widgetClassPath=&TextBlock
   parent=EditorUtilityButton_FixupLegacy
   widget=TextBlock_152
   properties {text=Update All Songs and Emote Item Defs}
   slot {class=ButtonSlot}
  }
 ]
}
